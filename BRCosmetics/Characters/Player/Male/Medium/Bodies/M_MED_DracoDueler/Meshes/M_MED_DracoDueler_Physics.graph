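# Physics asset for the DracoDueler medium male player body, covering only the pelvis, thighs and dynamic belt bones.
# A center belt pair links to the pelvis and branches through left and right master bones into three-box chains, with capsules on the pelvis and thighs.
# Chain joints have swing and twist limits, while thigh, anchor and master joints rotate freely.
# An extra free-rotation constraint ties the first left and right belt segments together.
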PhysicsAsset /BRCosmetics/Characters/Player/Male/Medium/Bodies/M_MED_DracoDueler/Meshes/M_MED_DracoDueler_Physics
bodies {
 dyn_belt_c [{shapeType=Capsule, length=2, radius=2, shapeName=dyn_belt_c_capsule}]
 dyn_belt_c_mstr [{shapeType=Capsule, length=2, radius=2, shapeName=dyn_belt_c_mstr_capsule}]
 dyn_belt_l_01 [{shapeType=Box, extentX=3.8807, extentY=1.4909, extentZ=8.1476, shapeName=dyn_belt_r_01_box, center=<4.1799 0 3>, rotation=<rot 0 90 -90>}]
 dyn_belt_l_02 [{
   shapeType=Box
   extentX=4.9698
   extentY=1.2035
   extentZ=10.7168
   shapeName=dyn_belt_l_02_box
   center=<4.9714 -0.3988 3>
   rotation=<rot 0 -88.7606 -88.4579>
  }]
 dyn_belt_l_03 [{
   shapeType=Box
   extentX=6.0206
   extentY=2.1816
   extentZ=8.9962
   shapeName=dyn_belt_l_03_box
   center=<5.0828 -0.222 1.6126>
   rotation=<rot 0 92.0416 -89.2576>
  }]
 dyn_belt_l_mstr [{shapeType=Capsule, length=0.505, radius=0.505, shapeName=dyn_belt_l_mstr_capsule, rotation=<rot 0 -90 0>}]
 dyn_belt_r_01 [{shapeType=Box, extentX=3.8807, extentY=1.4909, extentZ=8.1476, shapeName=dyn_belt_r_01_box, center=<4.1799 0 3>, rotation=<rot 0 90 -90>}]
 dyn_belt_r_02 [{
   shapeType=Box
   extentX=4.6569
   extentY=1.2816
   extentZ=9.3786
   shapeName=dyn_belt_r_02_box
   center=<4.8906 0.5089 3>
   rotation=<rot 0 -89.0012 -89.0227>
  }]
 dyn_belt_r_03 [{
   shapeType=Box
   extentX=6.4887
   extentY=1.2733
   extentZ=6.5726
   shapeName=dyn_belt_r_03_box
   center=<3.0121 0.6459 3>
   rotation=<rot 0.43 178.1592 -89.5663>
  }]
 dyn_belt_r_mstr [{shapeType=Capsule, length=0.505, radius=0.505, shapeName=dyn_belt_r_mstr_capsule, rotation=<rot 0 -90 0>}]
 pelvis [{shapeType=Capsule, length=21.7681, radius=14.1047, center=<-3.6652 0 0>, rotation=<rot 0 90 0>}]
 thigh_l [{shapeType=Capsule, length=18.8524, radius=11.904, shapeName=thigh_l_capsule, center=<-18.9732 0 0>, rotation=<rot 90 0 0>}]
 thigh_r [{shapeType=Capsule, length=19.6345, radius=11.1218, shapeName=thigh_l_capsule, center=<20.0992 0 0>, rotation=<rot 90 0 0>}]
}
constraints [
 {
  bone1Name=thigh_l
  bone2Name=pelvis
  swing1LimitDegrees=45
  swing1Motion=Free
  swing2LimitDegrees=45
  swing2Motion=Free
  twistLimitDegrees=45
  twistMotion=Free
 }
 {
  bone1Name=thigh_r
  bone2Name=pelvis
  swing1LimitDegrees=45
  swing1Motion=Free
  swing2LimitDegrees=45
  swing2Motion=Free
  twistLimitDegrees=45
  twistMotion=Free
 }
 {
  bone1Name=dyn_belt_l_mstr
  bone2Name=dyn_belt_c
  swing1LimitDegrees=45
  swing1Motion=Free
  swing2LimitDegrees=45
  swing2Motion=Free
  twistLimitDegrees=45
  twistMotion=Free
 }
 {bone1Name=dyn_belt_l_01, bone2Name=dyn_belt_l_mstr, swing1LimitDegrees=30, swing2LimitDegrees=60, twistLimitDegrees=30}
 {bone1Name=dyn_belt_l_02, bone2Name=dyn_belt_l_01, swing1LimitDegrees=30, swing2LimitDegrees=50, twistLimitDegrees=30}
 {bone1Name=dyn_belt_l_03, bone2Name=dyn_belt_l_02, swing1LimitDegrees=30, swing2LimitDegrees=50, twistLimitDegrees=30}
 {
  bone1Name=dyn_belt_r_mstr
  bone2Name=dyn_belt_c
  swing1LimitDegrees=45
  swing1Motion=Free
  swing2LimitDegrees=45
  swing2Motion=Free
  twistLimitDegrees=45
  twistMotion=Free
 }
 {bone1Name=dyn_belt_r_01, bone2Name=dyn_belt_r_mstr, swing1LimitDegrees=30, swing2LimitDegrees=70, twistLimitDegrees=30}
 {bone1Name=dyn_belt_r_02, bone2Name=dyn_belt_r_01, swing1LimitDegrees=30, swing2LimitDegrees=50, twistLimitDegrees=30}
 {bone1Name=dyn_belt_r_03, bone2Name=dyn_belt_r_02, swing1LimitDegrees=30, swing2LimitDegrees=50, twistLimitDegrees=30}
 {
  bone1Name=dyn_belt_r_01
  bone2Name=dyn_belt_l_01
  swing1LimitDegrees=45
  swing1Motion=Free
  swing2LimitDegrees=45
  swing2Motion=Free
  twistLimitDegrees=45
  twistMotion=Free
 }
 {
  bone1Name=dyn_belt_c_mstr
  bone2Name=pelvis
  swing1LimitDegrees=45
  swing1Motion=Free
  swing2LimitDegrees=45
  swing2Motion=Free
  twistLimitDegrees=45
  twistMotion=Free
 }
 {
  bone1Name=dyn_belt_c
  bone2Name=dyn_belt_c_mstr
  swing1LimitDegrees=45
  swing1Motion=Free
  swing2LimitDegrees=45
  swing2Motion=Free
  twistLimitDegrees=45
  twistMotion=Free
 }
]
