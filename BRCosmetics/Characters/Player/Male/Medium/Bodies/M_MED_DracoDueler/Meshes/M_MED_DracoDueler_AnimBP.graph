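# Body AnimBP for the DracoDueler cosmetic character, adding secondary motion to collar, belts, star and shoulder pads.
# It copies the mesh pose, takes jaw bones from the face, then runs constraints, Control Rig, AnimDynamics and RigidBody into a cached pose.
# With rigid body disabled, belts follow thigh constraints, using non-twist bones at LOD 2+.
# The event graph updates these switches and jaw rotation, and hides tails under capes.
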AnimBlueprint /BRCosmetics/Characters/Player/Male/Medium/Bodies/M_MED_DracoDueler/Meshes/M_MED_DracoDueler_AnimBP
target_skeleton=&/M_MED_DracoDueler_Skeleton
graphs [
 AnimGraph {nodes [
   n1 = "Constraint - dyn_shoulderpad_l_mstr (PPPP)"!(
    ComponentPose
    Alpha=1
    ConstraintWeights_0=0.42
    ConstraintWeights_1:float=0.47
    ConstraintWeights_2:float=0.1
    ConstraintWeights_3:float=0.01
   ) -> (Pose) {
    alphaBoolBlend {blendOption=Linear}
    boneToModify {boneName=dyn_shoulderpad_l_mstr}
    constraintSetup [
     {targetBone {boneName=upperarm_twist_01_l}}
     {targetBone {boneName=clavicle_l}}
     {targetBone {boneName=spine_05}}
     {targetBone {boneName=neck_02}}
    ]
    constraintWeights [1,1,1,1]
   }
   "Constraint - dyn_shoulderpad_r_mstr (PPPPP)"!(
    ComponentPose=@n1
    Alpha=1
    ConstraintWeights_0=0.27
    ConstraintWeights_1:float=0.59
    ConstraintWeights_2:float=0.1
    ConstraintWeights_3:float=0.01
    ConstraintWeights_4:float=0.03
   ) -> (Pose) {
    alphaBoolBlend {blendOption=Linear}
    boneToModify {boneName=dyn_shoulderpad_r_mstr}
    constraintSetup [
     {targetBone {boneName=upperarm_twist_01_r}}
     {targetBone {boneName=clavicle_r}}
     {targetBone {boneName=spine_05}}
     {targetBone {boneName=neck_02}}
     {targetBone {boneName=spine_03}}
    ]
    constraintWeights [1,1,1,1,1]
   }
   n2 = "Constraint - dyn_belt_r_mstr (PPPP)"!(
    ComponentPose
    Alpha=1
    ConstraintWeights_0=0.3394
    ConstraintWeights_1:float=0.0472
    ConstraintWeights_2:float=0.5948
    ConstraintWeights_3:float=0.0186
   ) -> (Pose) {
    alphaBoolBlend {blendOption=Linear}
    boneToModify {boneName=dyn_belt_r_mstr}
    constraintSetup [{targetBone {boneName=spine_01}}, {targetBone {boneName=spine_02}}, {targetBone {boneName=pelvis}}, {targetBone {boneName=spine_03}}]
    constraintWeights [1,1,1,1]
   }
   "Constraint - dyn_belt_l_mstr (PPPP)"!(
    ComponentPose=@n2
    Alpha=1
    ConstraintWeights_0=0.3526
    ConstraintWeights_1:float=0.0497
    ConstraintWeights_2:float=0.5756
    ConstraintWeights_3:float=0.0221
   ) -> (Pose) {
    alphaBoolBlend {blendOption=Linear}
    boneToModify {boneName=dyn_belt_l_mstr}
    constraintSetup [{targetBone {boneName=spine_01}}, {targetBone {boneName=spine_02}}, {targetBone {boneName=pelvis}}, {targetBone {boneName=spine_03}}]
    constraintWeights [1,1,1,1]
   }
   n3 = "Constraint - dyn_belt_l_01 (PP)"!(ComponentPose=Local To Component(LocalPose=Use cached pose 'Pre_Pose'()), Alpha=1, ConstraintWeights_0=1, ConstraintWeights_1:float=0) -> (Pose) {
    alphaBoolBlend {blendOption=Linear}
    boneToModify {boneName=dyn_belt_l_01}
    constraintSetup [{targetBone {boneName=dyn_belt_c}}, {targetBone {boneName=thigh_l}}]
    constraintWeights [1,1]
   }
   n4 = "Constraint - dyn_belt_r_01 (PP)"!(ComponentPose=@n3, Alpha=1, ConstraintWeights_0=1, ConstraintWeights_1:float=0) -> (Pose) {
    alphaBoolBlend {blendOption=Linear}
    boneToModify {boneName=dyn_belt_r_01}
    constraintSetup [{targetBone {boneName=dyn_belt_c}}, {targetBone {boneName=thigh_r}}]
    constraintWeights [1,1]
   }
   n5 = "Constraint - dyn_belt_l_01 (RR)"!(ComponentPose=@n4, Alpha=1, ConstraintWeights_0=0, ConstraintWeights_1:float=1) -> (Pose) {
    alphaBoolBlend {blendOption=Linear}
    boneToModify {boneName=dyn_belt_l_01}
    constraintSetup [{transformType=Rotation, targetBone {boneName=dyn_belt_c}}, {transformType=Rotation, targetBone {boneName=thigh_twist_01_l}}]
    constraintWeights [1,1]
   }
   n6 = "Constraint - dyn_belt_r_01 (RR)"!(ComponentPose=@n5, Alpha=1, ConstraintWeights_0=0, ConstraintWeights_1:float=1) -> (Pose) {
    alphaBoolBlend {blendOption=Linear}
    boneToModify {boneName=dyn_belt_r_01}
    constraintSetup [{transformType=Rotation, targetBone {boneName=dyn_belt_c}}, {transformType=Rotation, targetBone {boneName=thigh_twist_01_r}}]
    constraintWeights [1,1]
   }
   n7 = "Transform (Modify) Bone - Bone: dyn_belt_l_01"(ComponentPose=@n6, Alpha=1, Translation="0,0,-0.5", Rotation="0,0,0", Scale="1,1,1") {
    rotationMode=BMM_Additive
    rotationSpace=BCS_BoneSpace
    translationMode=BMM_Additive
    translationSpace=BCS_BoneSpace
    alphaBoolBlend {blendOption=Linear}
    boneToModify {boneName=dyn_belt_l_01}
    translation {z=-0.5}
   }
   n8 = "Transform (Modify) Bone - Bone: dyn_belt_r_01"(ComponentPose=@n7, Alpha=1, Translation="0,0,-0.5", Rotation="0,0,0", Scale="1,1,1") {
    rotationMode=BMM_Additive
    rotationSpace=BCS_BoneSpace
    translationMode=BMM_Additive
    translationSpace=BCS_BoneSpace
    alphaBoolBlend {blendOption=Linear}
    boneToModify {boneName=dyn_belt_r_01}
    translation {z=-0.5}
   }
   n9 = "Constraint - dyn_belt_l_02 (PP)"!(ComponentPose=@n8, Alpha=1, ConstraintWeights_0=0.5, ConstraintWeights_1:float=0.5) -> (Pose) {
    alphaBoolBlend {blendOption=Linear}
    boneToModify {boneName=dyn_belt_l_02}
    constraintSetup [{targetBone {boneName=thigh_twist_01_l}}, {targetBone {boneName=dyn_belt_l_01}}]
    constraintWeights [1,1]
   }
   n10 = "Constraint - dyn_belt_r_02 (PP)"!(ComponentPose=@n9, Alpha=1, ConstraintWeights_0=0.5, ConstraintWeights_1:float=0.5) -> (Pose) {
    alphaBoolBlend {blendOption=Linear}
    boneToModify {boneName=dyn_belt_r_02}
    constraintSetup [{targetBone {boneName=thigh_twist_01_r}}, {targetBone {boneName=dyn_belt_r_01}}]
    constraintWeights [1,1]
   }
   n11 = "Constraint - dyn_belt_l_03 (PP)"!(ComponentPose=@n10, Alpha=1, ConstraintWeights_0=0.5, ConstraintWeights_1:float=0.5) -> (Pose) {
    alphaBoolBlend {blendOption=Linear}
    boneToModify {boneName=dyn_belt_l_03}
    constraintSetup [{targetBone {boneName=thigh_twist_01_l}}, {targetBone {boneName=dyn_belt_l_02}}]
    constraintWeights [1,1]
   }
   n12 = "Constraint - dyn_belt_r_03 (PP)"!(ComponentPose=@n11, Alpha=1, ConstraintWeights_0=0.5, ConstraintWeights_1:float=0.5) -> (Pose) {
    alphaBoolBlend {blendOption=Linear}
    boneToModify {boneName=dyn_belt_r_03}
    constraintSetup [{targetBone {boneName=thigh_twist_01_r}}, {targetBone {boneName=dyn_belt_r_02}}]
    constraintWeights [1,1]
   }
   n13 = "Constraint - dyn_belt_l_04 (P)"(ComponentPose=@n12, Alpha=1, ConstraintWeights_0=1) {
    alphaBoolBlend {blendOption=Linear}
    boneToModify {boneName=dyn_belt_l_04}
    constraintSetup [{targetBone {boneName=thigh_twist_01_l}}]
    constraintWeights [1]
   }
   n14 = "Constraint - dyn_belt_r_04 (P)"(ComponentPose=@n13, Alpha=1, ConstraintWeights_0=1) {
    alphaBoolBlend {blendOption=Linear}
    boneToModify {boneName=dyn_belt_r_04}
    constraintSetup [{targetBone {boneName=thigh_twist_01_r}}]
    constraintWeights [1]
   }
   n15 = "Transform (Modify) Bone - Bone: dyn_belt_l_02"(ComponentPose=@n14, Alpha=1, Translation="0,0,-1", Rotation="0,0,0", Scale="1,1,1") {
    rotationMode=BMM_Additive
    rotationSpace=BCS_BoneSpace
    translationMode=BMM_Additive
    translationSpace=BCS_BoneSpace
    alphaBoolBlend {blendOption=Linear}
    boneToModify {boneName=dyn_belt_l_02}
    translation {z=-1}
   }
   n16 = "Transform (Modify) Bone - Bone: dyn_belt_r_02"(ComponentPose=@n15, Alpha=1, Translation="0,0,-1", Rotation="0,0,0", Scale="1,1,1") {
    rotationMode=BMM_Additive
    rotationSpace=BCS_BoneSpace
    translationMode=BMM_Additive
    translationSpace=BCS_BoneSpace
    alphaBoolBlend {blendOption=Linear}
    boneToModify {boneName=dyn_belt_r_02}
    translation {z=-1}
   }
   n17 = "Transform (Modify) Bone - Bone: dyn_belt_l_03"(ComponentPose=@n16, Alpha=1, Translation="0,0,-3", Rotation="0,0,0", Scale="1,1,1") {
    rotationMode=BMM_Additive
    rotationSpace=BCS_BoneSpace
    translationMode=BMM_Additive
    translationSpace=BCS_BoneSpace
    alphaBoolBlend {blendOption=Linear}
    boneToModify {boneName=dyn_belt_l_03}
    translation {z=-3}
   }
   n18 = "Transform (Modify) Bone - Bone: dyn_belt_r_03"(ComponentPose=@n17, Alpha=1, Translation="0,0,-2", Rotation="0,0,0", Scale="1,1,1") {
    rotationMode=BMM_Additive
    rotationSpace=BCS_BoneSpace
    translationMode=BMM_Additive
    translationSpace=BCS_BoneSpace
    alphaBoolBlend {blendOption=Linear}
    boneToModify {boneName=dyn_belt_r_03}
    translation {z=-2}
   }
   n19 = Component To Local(ComponentPose=@n18)
   n20 = "Constraint - dyn_belt_l_01 (PP)"(ComponentPose=Local To Component(LocalPose=Use cached pose 'Pre_Pose'()), Alpha=1, ConstraintWeights_0=1, ConstraintWeights_1=0) {
    alphaBoolBlend {blendOption=Linear}
    boneToModify {boneName=dyn_belt_l_01}
    constraintSetup [{targetBone {boneName=dyn_belt_c}}, {targetBone {boneName=thigh_l}}]
    constraintWeights [1,1]
   }
   n21 = "Constraint - dyn_belt_r_01 (PP)"(ComponentPose=@n20, Alpha=1, ConstraintWeights_0=1, ConstraintWeights_1=0) {
    alphaBoolBlend {blendOption=Linear}
    boneToModify {boneName=dyn_belt_r_01}
    constraintSetup [{targetBone {boneName=dyn_belt_c}}, {targetBone {boneName=thigh_r}}]
    constraintWeights [1,1]
   }
   n22 = "Constraint - dyn_belt_l_01 (RR)"(ComponentPose=@n21, Alpha=1, ConstraintWeights_0=0, ConstraintWeights_1=1) {
    alphaBoolBlend {blendOption=Linear}
    boneToModify {boneName=dyn_belt_l_01}
    constraintSetup [{transformType=Rotation, targetBone {boneName=dyn_belt_c}}, {transformType=Rotation, targetBone {boneName=thigh_l}}]
    constraintWeights [1,1]
   }
   n23 = "Constraint - dyn_belt_r_01 (RR)"(ComponentPose=@n22, Alpha=1, ConstraintWeights_0=0, ConstraintWeights_1=1) {
    alphaBoolBlend {blendOption=Linear}
    boneToModify {boneName=dyn_belt_r_01}
    constraintSetup [{transformType=Rotation, targetBone {boneName=dyn_belt_c}}, {transformType=Rotation, targetBone {boneName=thigh_r}}]
    constraintWeights [1,1]
   }
   n24 = "Transform (Modify) Bone - Bone: dyn_belt_l_01"(ComponentPose=@n23, Alpha=1, Translation="0,0,-0.5", Rotation="0,0,0", Scale="1,1,1") {
    rotationMode=BMM_Additive
    rotationSpace=BCS_BoneSpace
    translationMode=BMM_Additive
    translationSpace=BCS_BoneSpace
    alphaBoolBlend {blendOption=Linear}
    boneToModify {boneName=dyn_belt_l_01}
    translation {z=-0.5}
   }
   n25 = "Transform (Modify) Bone - Bone: dyn_belt_r_01"(ComponentPose=@n24, Alpha=1, Translation="0,0,-0.5", Rotation="0,0,0", Scale="1,1,1") {
    rotationMode=BMM_Additive
    rotationSpace=BCS_BoneSpace
    translationMode=BMM_Additive
    translationSpace=BCS_BoneSpace
    alphaBoolBlend {blendOption=Linear}
    boneToModify {boneName=dyn_belt_r_01}
    translation {z=-0.5}
   }
   n26 = "Constraint - dyn_belt_l_02 (PP)"(ComponentPose=@n25, Alpha=1, ConstraintWeights_0=0.5, ConstraintWeights_1=0.5) {
    alphaBoolBlend {blendOption=Linear}
    boneToModify {boneName=dyn_belt_l_02}
    constraintSetup [{targetBone {boneName=thigh_l}}, {targetBone {boneName=dyn_belt_l_01}}]
    constraintWeights [1,1]
   }
   n27 = "Constraint - dyn_belt_r_02 (PP)"(ComponentPose=@n26, Alpha=1, ConstraintWeights_0=0.5, ConstraintWeights_1=0.5) {
    alphaBoolBlend {blendOption=Linear}
    boneToModify {boneName=dyn_belt_r_02}
    constraintSetup [{targetBone {boneName=thigh_r}}, {targetBone {boneName=dyn_belt_r_01}}]
    constraintWeights [1,1]
   }
   n28 = "Constraint - dyn_belt_l_03 (PP)"(ComponentPose=@n27, Alpha=1, ConstraintWeights_0=0.5, ConstraintWeights_1=0.5) {
    alphaBoolBlend {blendOption=Linear}
    boneToModify {boneName=dyn_belt_l_03}
    constraintSetup [{targetBone {boneName=thigh_l}}, {targetBone {boneName=dyn_belt_l_02}}]
    constraintWeights [1,1]
   }
   n29 = "Constraint - dyn_belt_r_03 (PP)"(ComponentPose=@n28, Alpha=1, ConstraintWeights_0=0.5, ConstraintWeights_1=0.5) {
    alphaBoolBlend {blendOption=Linear}
    boneToModify {boneName=dyn_belt_r_03}
    constraintSetup [{targetBone {boneName=thigh_r}}, {targetBone {boneName=dyn_belt_r_02}}]
    constraintWeights [1,1]
   }
   n30 = "Constraint - dyn_belt_l_04 (P)"(ComponentPose=@n29, Alpha=1, ConstraintWeights_0=1) {
    alphaBoolBlend {blendOption=Linear}
    boneToModify {boneName=dyn_belt_l_04}
    constraintSetup [{targetBone {boneName=thigh_l}}]
    constraintWeights [1]
   }
   n31 = "Constraint - dyn_belt_r_04 (P)"(ComponentPose=@n30, Alpha=1, ConstraintWeights_0=1) {
    alphaBoolBlend {blendOption=Linear}
    boneToModify {boneName=dyn_belt_r_04}
    constraintSetup [{targetBone {boneName=thigh_r}}]
    constraintWeights [1]
   }
   n32 = "Transform (Modify) Bone - Bone: dyn_belt_l_02"(ComponentPose=@n31, Alpha=1, Translation="0,0,-1", Rotation="0,0,0", Scale="1,1,1") {
    rotationMode=BMM_Additive
    rotationSpace=BCS_BoneSpace
    translationMode=BMM_Additive
    translationSpace=BCS_BoneSpace
    alphaBoolBlend {blendOption=Linear}
    boneToModify {boneName=dyn_belt_l_02}
    translation {z=-1}
   }
   n33 = "Transform (Modify) Bone - Bone: dyn_belt_r_02"(ComponentPose=@n32, Alpha=1, Translation="0,0,-1", Rotation="0,0,0", Scale="1,1,1") {
    rotationMode=BMM_Additive
    rotationSpace=BCS_BoneSpace
    translationMode=BMM_Additive
    translationSpace=BCS_BoneSpace
    alphaBoolBlend {blendOption=Linear}
    boneToModify {boneName=dyn_belt_r_02}
    translation {z=-1}
   }
   n34 = "Transform (Modify) Bone - Bone: dyn_belt_l_03"(ComponentPose=@n33, Alpha=1, Translation="0,0,-3", Rotation="0,0,0", Scale="1,1,1") {
    rotationMode=BMM_Additive
    rotationSpace=BCS_BoneSpace
    translationMode=BMM_Additive
    translationSpace=BCS_BoneSpace
    alphaBoolBlend {blendOption=Linear}
    boneToModify {boneName=dyn_belt_l_03}
    translation {z=-3}
   }
   n35 = "Transform (Modify) Bone - Bone: dyn_belt_r_03"(ComponentPose=@n34, Alpha=1, Translation="0,0,-2", Rotation="0,0,0", Scale="1,1,1") {
    rotationMode=BMM_Additive
    rotationSpace=BCS_BoneSpace
    translationMode=BMM_Additive
    translationSpace=BCS_BoneSpace
    alphaBoolBlend {blendOption=Linear}
    boneToModify {boneName=dyn_belt_r_03}
    translation {z=-2}
   }
   n36 = Component To Local(ComponentPose=@n35)
   n37 = Get lod_2_up() -> (lod_2_up:bool)
   Output Pose(Result=Blend Poses by bool(
     BlendPose_0=Use cached pose 'Pre_Pose'()
     BlendPose_1=Blend Poses by bool(BlendPose_0=@n36, BlendPose_1=@n19, bActiveValue=@n37) {blendType=Linear}
     bActiveValue=Get rban_enabled() -> (rban_enabled:bool)
    ) {blendType=Linear}) {layerGroup=None}
   n38 = Blend Bone By Channel!(
    A:Pose Link Structure=Copy Pose From Mesh() {bCopyCurves=true, bCopyCustomAttributes=true, bUseAttachedParent=true, bUseMeshPose=true}
    B:Pose Link Structure=Copy Pose From Mesh(SourceMeshComponent=Get Face_mesh() -> (Face_mesh:Skeletal Mesh Component Object Reference)) {bCopyCurves=true, bCopyCustomAttributes=true, bUseAttachedParent=true, bUseMeshPose=true}
    Alpha:float=1
   ) -> (Pose:Pose Link Structure) {
    is=title AnimGraphNode_BlendBoneByChannel
    blendNode {boneDefinitions [
      {bBlendRotation=true, bBlendScale=true, bBlendTranslation=true, sourceBone {boneName=dyn_jaw_mstr}, targetBone {boneName=dyn_jaw_mstr}}
      {bBlendRotation=true, bBlendScale=true, bBlendTranslation=true, sourceBone {boneName=dyn_jaw}, targetBone {boneName=dyn_jaw}}
     ]}
   }
   n39 = "Constraint - dyn_collar_front_r_mstr (PPPPPPPPP)"!(
    ComponentPose=Local To Component(LocalPose=@n38)
    Alpha=1
    ConstraintWeights_0=0.0084
    ConstraintWeights_1:float=0.0496
    ConstraintWeights_2:float=0.354
    ConstraintWeights_3:float=0
    ConstraintWeights_4:float=0.024
    ConstraintWeights_5:float=0.1517
    ConstraintWeights_6:float=0.0185
    ConstraintWeights_7:float=0.3917
    ConstraintWeights_8:float=0.002
   ) -> (Pose) {
    alphaBoolBlend {blendOption=Linear}
    boneToModify {boneName=dyn_collar_front_r_mstr}
    constraintSetup [
     {targetBone {boneName=upperarm_twist_01_r}}
     {targetBone {boneName=spine_04}}
     {targetBone {boneName=spine_05}}
     {targetBone {boneName=clavicle_l}}
     {targetBone {boneName=spine_03}}
     {targetBone {boneName=clavicle_r}}
     {targetBone {boneName=neck_01}}
     {targetBone {boneName=dyn_collar_front_r_01}}
     {targetBone {boneName=dyn_collar_back_r_01}}
    ]
    constraintWeights [1,1,1,1,1,1,1,1,1]
   }
   n40 = "Constraint - dyn_collar_back_r_mstr (PPPPPPPPPPPP)"!(
    ComponentPose=@n39
    Alpha=1
    ConstraintWeights_0=0
    ConstraintWeights_1:float=0.0061
    ConstraintWeights_2:float=0.0157
    ConstraintWeights_3:float=0.3597
    ConstraintWeights_4:float=0
    ConstraintWeights_5:float=0.0613
    ConstraintWeights_6:float=0.257
    ConstraintWeights_7:float=0.005
    ConstraintWeights_8:float=0
    ConstraintWeights_9:float=0.0403
    ConstraintWeights_10:float=0.2544
    ConstraintWeights_11:float=0.0005
   ) -> (Pose) {
    alphaBoolBlend {blendOption=Linear}
    boneToModify {boneName=dyn_collar_back_r_mstr}
    constraintSetup [
     {targetBone {boneName=upperarm_twist_01_l}}
     {targetBone {boneName=upperarm_twist_01_r}}
     {targetBone {boneName=spine_04}}
     {targetBone {boneName=spine_05}}
     {targetBone {boneName=clavicle_l}}
     {targetBone {boneName=spine_03}}
     {targetBone {boneName=clavicle_r}}
     {targetBone {boneName=neck_01}}
     {targetBone {boneName=dyn_collar_front_l_01}}
     {targetBone {boneName=dyn_collar_front_r_01}}
     {targetBone {boneName=dyn_collar_back_r_01}}
     {targetBone {boneName=dyn_collar_back_l_01}}
    ]
    constraintWeights [1,1,1,1,1,1,1,1,1,1,1,1]
   }
   n41 = "Constraint - dyn_collar_front_l_mstr (PPPPPPPPPPPP)"!(
    ComponentPose=@n40
    Alpha=1
    ConstraintWeights_0=0.0051
    ConstraintWeights_1:float=0
    ConstraintWeights_2:float=0.044
    ConstraintWeights_3:float=0.3641
    ConstraintWeights_4:float=0.1322
    ConstraintWeights_5:float=0.0258
    ConstraintWeights_6:float=0
    ConstraintWeights_7:float=0.0214
    ConstraintWeights_8:float=0.3996
    ConstraintWeights_9:float=0
    ConstraintWeights_10:float=0.0075
    ConstraintWeights_11:float=0.0002
   ) -> (Pose) {
    alphaBoolBlend {blendOption=Linear}
    boneToModify {boneName=dyn_collar_front_l_mstr}
    constraintSetup [
     {targetBone {boneName=upperarm_twist_01_l}}
     {targetBone {boneName=upperarm_twist_01_r}}
     {targetBone {boneName=spine_04}}
     {targetBone {boneName=spine_05}}
     {targetBone {boneName=clavicle_l}}
     {targetBone {boneName=spine_03}}
     {targetBone {boneName=clavicle_r}}
     {targetBone {boneName=neck_01}}
     {targetBone {boneName=dyn_collar_front_l_01}}
     {targetBone {boneName=dyn_collar_back_r_01}}
     {targetBone {boneName=dyn_collar_back_l_01}}
     {targetBone {boneName=dyn_shoulderpad_l}}
    ]
    constraintWeights [1,1,1,1,1,1,1,1,1,1,1,1]
   }
   n42 = "Constraint - dyn_collar_back_l_mstr (PPPPPPPPPPP)"!(
    ComponentPose=@n41
    Alpha=1
    ConstraintWeights_0=0.0006
    ConstraintWeights_1:float=0
    ConstraintWeights_2:float=0.0005
    ConstraintWeights_3:float=0.1478
    ConstraintWeights_4:float=0.1188
    ConstraintWeights_5:float=0.0227
    ConstraintWeights_6:float=0
    ConstraintWeights_7:float=0.004
    ConstraintWeights_8:float=0.0071
    ConstraintWeights_9:float=0.0003
    ConstraintWeights_10:float=0.6983
   ) -> (Pose) {
    alphaBoolBlend {blendOption=Linear}
    boneToModify {boneName=dyn_collar_back_l_mstr}
    constraintSetup [
     {targetBone {boneName=upperarm_twist_01_l}}
     {targetBone {boneName=upperarm_twist_01_r}}
     {targetBone {boneName=spine_04}}
     {targetBone {boneName=spine_05}}
     {targetBone {boneName=clavicle_l}}
     {targetBone {boneName=spine_03}}
     {targetBone {boneName=clavicle_r}}
     {targetBone {boneName=neck_01}}
     {targetBone {boneName=dyn_collar_front_l_01}}
     {targetBone {boneName=dyn_collar_back_r_01}}
     {targetBone {boneName=dyn_collar_back_l_01}}
    ]
    constraintWeights [1,1,1,1,1,1,1,1,1,1,1]
   }
   n43 = "Constraint - dyn_belt_c_mstr (PPPP)"!(
    ComponentPose=@n42
    Alpha=1
    ConstraintWeights_0=0.3509
    ConstraintWeights_1:float=0.048
    ConstraintWeights_2:float=0.5801
    ConstraintWeights_3:float=0.021
   ) -> (Pose) {
    alphaBoolBlend {blendOption=Linear}
    boneToModify {boneName=dyn_belt_c_mstr}
    constraintSetup [{targetBone {boneName=spine_01}}, {targetBone {boneName=spine_02}}, {targetBone {boneName=pelvis}}, {targetBone {boneName=spine_03}}]
    constraintWeights [1,1,1,1]
   }
   n44 = "Constraint - dyn_belt_r_mstr (PPPP)"(ComponentPose=@n43, Alpha=1, ConstraintWeights_0=0.3394, ConstraintWeights_1=0.0472, ConstraintWeights_2=0.5948, ConstraintWeights_3=0.0186) {
    alphaBoolBlend {blendOption=Linear}
    boneToModify {boneName=dyn_belt_r_mstr}
    constraintSetup [{targetBone {boneName=spine_01}}, {targetBone {boneName=spine_02}}, {targetBone {boneName=pelvis}}, {targetBone {boneName=spine_03}}]
    constraintWeights [1,1,1,1]
   }
   n45 = "Constraint - dyn_belt_l_mstr (PPPP)"(ComponentPose=@n44, Alpha=1, ConstraintWeights_0=0.3526, ConstraintWeights_1=0.0497, ConstraintWeights_2=0.5756, ConstraintWeights_3=0.0221) {
    alphaBoolBlend {blendOption=Linear}
    boneToModify {boneName=dyn_belt_l_mstr}
    constraintSetup [{targetBone {boneName=spine_01}}, {targetBone {boneName=spine_02}}, {targetBone {boneName=pelvis}}, {targetBone {boneName=spine_03}}]
    constraintWeights [1,1,1,1]
   }
   n46 = Control Rig!(Source=Component To Local(ComponentPose=@n45), Alpha, jaw_rot:double=Get jaw_rot() -> (jaw_rot:double)) -> (Pose) {
    alphaBoolBlend {blendOption=Linear}
    controlRigAssetReference {blueprintRigClass=&/M_MED_DracoDueler_CtrlRig.M_MED_DracoDueler_CtrlRig_C}
   }
   n47 = "Constraint - dyn_star_upper_mstr (PPPP)"!(
    ComponentPose=Local To Component(LocalPose=@n46)
    Alpha=1
    ConstraintWeights_0=0.0594
    ConstraintWeights_1:float=0.5256
    ConstraintWeights_2:float=0.228
    ConstraintWeights_3:float=0.187
   ) -> (Pose) {
    alphaBoolBlend {blendOption=Linear}
    boneToModify {boneName=dyn_star_upper_mstr}
    constraintSetup [{targetBone {boneName=spine_02}}, {targetBone {boneName=spine_04}}, {targetBone {boneName=spine_05}}, {targetBone {boneName=spine_03}}]
    constraintWeights [1,1,1,1]
   }
   n48 = "Constraint - dyn_star_lower_mstr (PPPP)"!(
    ComponentPose=@n47
    Alpha=1
    ConstraintWeights_0=0.0745
    ConstraintWeights_1:float=0.5491
    ConstraintWeights_2:float=0.1439
    ConstraintWeights_3:float=0.2325
   ) -> (Pose) {
    alphaBoolBlend {blendOption=Linear}
    boneToModify {boneName=dyn_star_lower_mstr}
    constraintSetup [{targetBone {boneName=spine_02}}, {targetBone {boneName=spine_04}}, {targetBone {boneName=spine_05}}, {targetBone {boneName=spine_03}}]
    constraintWeights [1,1,1,1]
   }
   n49 = Anim Dynamics - Bone: dyn_star_lower(ComponentPose=@n48, Alpha=1, ExternalForce="0,0,0") {
    angularDampingOverride=0.9
    angularSpringConstant=150
    bAngularSpring=true
    bGravityOverrideInSimSpace=true
    bOverrideAngularDamping=true
    bOverrideLinearDamping=true
    bUseGravityOverride=true
    gravityScale=0.35
    linearDampingOverride=0.999
    linearSpringConstant=150
    simulationSpace=RootRelative
    alphaBoolBlend {blendOption=Linear}
    boundBone {boneName=dyn_star_lower}
    componentAppliedLinearAccClamp=<1500 1500 1500>
    componentLinearAccScale=<0.1 0.1 0.25>
    componentLinearVelScale=<0.3 0.3 0.4>
    gravityOverride {z=-980}
    physicsBodyDefinitions [{
      sphereCollisionRadius=0
      boundBone {boneName=dyn_star_lower}
      boxExtents=<10 9 10>
      constraintSetup {angularLimitsMax=<5 10 10>, angularLimitsMin=<-5 -10 -10>, angularTarget=<1 0 0>}
      localJointOffset=<0.65 0 0>
     }]
   }
   n50 = Anim Dynamics - Bone: dyn_star_upper(ComponentPose=@n49, Alpha=1, ExternalForce="0,0,0") {
    angularDampingOverride=0.9
    angularSpringConstant=150
    bAngularSpring=true
    bGravityOverrideInSimSpace=true
    bOverrideAngularDamping=true
    bOverrideLinearDamping=true
    bUseGravityOverride=true
    gravityScale=0.35
    linearDampingOverride=0.999
    linearSpringConstant=150
    simulationSpace=RootRelative
    alphaBoolBlend {blendOption=Linear}
    boundBone {boneName=dyn_star_upper}
    componentAppliedLinearAccClamp=<1500 1500 1500>
    componentLinearAccScale=<0.1 0.1 0.25>
    componentLinearVelScale=<0.3 0.3 0.4>
    gravityOverride {z=-980}
    physicsBodyDefinitions [{
      sphereCollisionRadius=0
      boundBone {boneName=dyn_star_upper}
      boxExtents=<10 9 10>
      constraintSetup {angularLimitsMax=<5 10 10>, angularLimitsMin=<-5 -10 -5>, angularTarget=<1 0 0>}
      localJointOffset=<0.65 0 0>
     }]
   }
   n51 = "Transform (Modify) Bone - Bone: dyn_star_upper_mstr"(ComponentPose=@n50, Alpha=1, Translation="0,0.2,0", Rotation="0,0,0", Scale="1,1,1") {
    rotationMode=BMM_Additive
    rotationSpace=BCS_BoneSpace
    translationMode=BMM_Additive
    translationSpace=BCS_BoneSpace
    alphaBoolBlend {blendOption=Linear}
    boneToModify {boneName=dyn_star_upper_mstr}
    translation {y=0.2}
   }
   n52 = RigidBody!(
    ComponentPose=@n51
    Alpha
    OverrideWorldGravity:Vector="0,0,-980"
    ExternalForce
    SimSpaceSettings:Sim Space Settings Structure=(WorldAlpha=0,VelocityScaleZ=0,DampingAlpha=1,MaxLinearVelocity=500,MaxAngularVelocity=3,MaxLinearAcceleration=700,MaxAngularAcceleration=1000,ExternalLinearDragV=(X=0.2,Y=0.2,Z=0),ExternalLinearVelocity=(X=0,Y=0,Z=0),ExternalAngularVelocity=(X=0,Y=0,Z=0))
   ) -> (Pose) {
    bClampLinearTranslationLimitToRefPose=true
    bOverrideWorldGravity=true
    overridePhysicsAsset=&/M_MED_DracoDueler_Physics
    simulationSpace=BaseBoneSpace
    alphaBoolBlend {blendInTime=0.1, blendOption=Linear, blendOutTime=0.1}
    baseBoneRef {boneName=pelvis}
    componentAppliedLinearAccClamp=<2000 2000 2000>
    componentLinearAccScale=<0.6 0.6 0.9>
    componentLinearVelScale=<0.6 0.6 0.9>
    simSpaceSettings {
     maxAngularAcceleration=1000
     maxAngularVelocity=3
     maxLinearAcceleration=700
     maxLinearVelocity=500
     velocityScaleZ=0
     externalLinearDragV=<0.2 0.2>
    }
   }
   n53 = "Transform (Modify) Bone - Bone: dyn_shoulderpad_r"!(ComponentPose=@n52, bAlphaBoolEnabled:bool=True, Translation="1,0,0.5", Rotation="7,0,0", Scale="1,1,1") -> (Pose) {
    alphaInputType=Bool
    rotationMode=BMM_Additive
    rotationSpace=BCS_BoneSpace
    translationMode=BMM_Additive
    translationSpace=BCS_BoneSpace
    alphaBoolBlend {blendOption=Linear}
    boneToModify {boneName=dyn_shoulderpad_r}
    rotation {pitch=7}
    translation {x=1, z=0.5}
   }
   n54 = "Transform (Modify) Bone - Bone: dyn_shoulderpad_l"!(ComponentPose=@n53, bAlphaBoolEnabled:bool=True, Translation="0.5,0,0", Rotation="0,0,0", Scale="1,1,1") -> (Pose) {
    alphaInputType=Bool
    rotationMode=BMM_Additive
    rotationSpace=BCS_BoneSpace
    translationMode=BMM_Additive
    translationSpace=BCS_BoneSpace
    alphaBoolBlend {blendOption=Linear}
    boneToModify {boneName=dyn_shoulderpad_l}
    translation {x=0.5}
   }
   n55 = "Transform (Modify) Bone - Bone: dyn_shoulderpad_rotate"!(ComponentPose=@n54, bAlphaBoolEnabled:bool=True, Translation="0,0,0", Rotation="0,0,-5", Scale="1,1,1") -> (Pose) {
    alphaInputType=Bool
    rotationMode=BMM_Additive
    rotationSpace=BCS_BoneSpace
    translationMode=BMM_Additive
    translationSpace=BCS_BoneSpace
    alphaBoolBlend {blendOption=Linear}
    boneToModify {boneName=dyn_shoulderpad_rotate}
    rotation {roll=-5}
   }
   n56 = "Transform (Modify) Bone - Bone: dyn_shoulderpad_r"(ComponentPose=@n55, bAlphaBoolEnabled=True, Translation="0,0,0", Rotation="0,0,2", Scale="1,1,1") {
    alphaInputType=Bool
    rotationMode=BMM_Additive
    rotationSpace=BCS_BoneSpace
    translationMode=BMM_Additive
    translationSpace=BCS_BoneSpace
    alphaBoolBlend {blendOption=Linear}
    boneToModify {boneName=dyn_shoulderpad_r}
    rotation {roll=2}
   }
   Save cached pose 'Pre_Pose'(Pose=Component To Local(ComponentPose=@n56)) {cacheName=Pre_Pose}
   Get Custom Idle is Playing Float() -> (Custom Idle is Playing Float:float)
  ]}
 EventGraph {nodes [
   Event BlueprintUpdateAnimation()
   > n1 = Sequence!(execute) -> (then_0, then_1, then_2:Exec, then_3:Exec, then_4:Exec, then_5:Exec, then_6:Exec, then_7:Exec, then_8:Exec)
   n2 = GetDeltaTransformFromRefPose(self=GetOwnerSkeletalMeshComponent(), BoneName=head, BaseName=spine_05)
   > Set headSpineRot(headSpineRot:double=BreakRotator(InRot=BreakTransform(InTransform=@n2).Rotation).Yaw) -> (Output_Get:double)
   Set head_rot_back(execute=@n1.then_1, head_rot_back:double=MapRangeClamped(Value=Get headSpineRot() -> (headSpineRot:double), InRangeA=-30, InRangeB=-60)) -> (Output_Get:double)
   Set Head_Rot_FAR_Back(execute=@n1.then_2, Head_Rot_FAR_Back:double=MapRangeClamped(Value=Get headSpineRot(), InRangeA=-60, InRangeB=-96)) -> (Output_Get:double)
   Set rban_enabled(execute=@n1.then_7, rban_enabled:bool=Get bRigidBodyEnabled()) -> (Output_Get:bool)
   > Set lod_2_up(lod_2_up:bool="integer >= integer"(A=Get CurrentLOD(), B=2)) -> (Output_Get:bool)
   > n3 = Sequence!(execute) -> (then_0, then_1)
   > Branch(Condition=Get bIsInFrontEnd())
   > Set Custom Idle is Playing Float(Custom Idle is Playing Float:float=GetCurveValue(self=GetOwnerFrontEndAnimInstance(), CurveName=Frontend_Idle_Bypass)) -> (Output_Get:float)
   Set jaw_rot(
    execute=@n3.then_1
    jaw_rot:double=MapRangeClamped(Value=Get Custom Idle is Playing Float(), InRangeA=0, InRangeB=1, OutRangeA=1, OutRangeB=0)
   ) -> (Output_Get:double)
   n4 = GetDeltaTransformFromRefPose(self=GetOwnerSkeletalMeshComponent(), BoneName=upperarm_r, BaseName=spine_05)
   Set UpArm_Spine_Rot(execute=@n1.then_3, UpArm_Spine_Rot:double=BreakRotator(InRot=BreakTransform(InTransform=@n4).Rotation).Yaw) -> (Output_Get:double)
   Set ADS_UpArm_Rot(execute=@n1.then_4, ADS_UpArm_Rot:double=MapRangeClamped(Value=Get UpArm_Spine_Rot() -> (UpArm_Spine_Rot:double), InRangeA=70, InRangeB=140)) -> (Output_Get:double)
   n5 = GetDeltaTransformFromRefPose(self=GetOwnerSkeletalMeshComponent(), BoneName=upperarm_r, BaseName=spine_05)
   Set UpArm_Spine_Rot_R(execute=@n1.then_5, UpArm_Spine_Rot_R:double=BreakRotator(InRot=BreakTransform(InTransform=@n5).Rotation).Yaw) -> (Output_Get:double)
   Set ADS_UpArm_Rot_R(
    execute=@n1.then_6
    ADS_UpArm_Rot_R:double=MapRangeClamped(Value=Get UpArm_Spine_Rot_R() -> (UpArm_Spine_Rot_R:double), InRangeA=70, InRangeB=140)
   ) -> (Output_Get:double)
   Event OnCharacterCustomizationLoaded()
   > Sequence()
   > Set BodyMesh(BodyMesh=GetOwnerSkeletalMeshComponent())
   > n6 = Cast To PlayerPawn_Athena(Object=TryGetPawnOwner())
   > n7 = Set PawnOwner(PawnOwner:Player Pawn Athena Object Reference=@n6) -> (Output_Get:Player Pawn Athena Object Reference)
   > Set TailMesh(TailMesh:Skeletal Mesh Component Object Reference=GetSkeletalMeshForPartType(self=@n7, PartType=MiscOrTail)) -> (Output_Get:Skeletal Mesh Component Object Reference)
   n8 = CheckForBackbling()
   > Set isWearingBackbling(isWearingBackbling=@n8)
   > Set isWearingCape(isWearingCape=@n8.bIsBackblingLong)
   > Set characterHasTail(characterHasTail:bool=CheckForTail!(self:Custom Character Part Anim Instance Object Reference) -> (bHasTail:bool)) -> (Output_Get:bool)
   > Branch(Condition=Get characterHasTail() -> (characterHasTail:bool))
   > n9 = Branch(Condition=Get isWearingCape() -> (isWearingCape:bool))
   n10 = Get TailMesh() -> (TailMesh:Skeletal Mesh Component Object Reference)
   > SetVisibility!(execute, self:Scene Component Object Reference=@n10, bNewVisibility:bool=false, bPropagateToChildren:bool=false) -> (then)
   > Set cape and tail combo(cape and tail combo:bool=false) -> (Output_Get:bool)
   > Set cape and tail(cape and tail:bool=false) -> (Output_Get:bool)
   SetVisibility(execute=@n9.else, self=@n10, bNewVisibility=true, bPropagateToChildren=false)
   > Set cape and tail(cape and tail=true)
   > n11 = Branch(Condition=Get bIsInFrontEnd())
   > Set Cape and Tail Combo Frontend(Cape and Tail Combo Frontend:bool=true) -> (Output_Get:bool)
   Set cape and tail combo(execute=@n11.else, cape and tail combo=true)
   Event BlueprintInitializeAnimation()
  ]}
]
variables [
 ADS_UpArm_Rot {type=number}
 ADS_UpArm_Rot_R {type=number}
 BodyMesh {type=SkeletalMeshComponent}
 cape and tail {type=boolean}
 cape and tail combo {type=boolean}
 Cape and Tail Combo Frontend {type=boolean}
 characterHasTail {type=boolean}
 Custom Idle is Playing Float {type=number}
 Face_mesh {type=SkeletalMeshComponent}
 head_rot_back {type=number}
 Head_Rot_FAR_Back {type=number}
 headSpineRot {type=number}
 isWearingBackbling {type=boolean}
 isWearingCape {type=boolean}
 jaw_rot {type=number}
 lod_2_up {type=boolean}
 Pawn Owner {type=FortPlayerPawnAthena}
 PawnOwner {type=PlayerPawn_Athena_C}
 rban_enabled {type=boolean}
 Skel_Mesh {type=SkeletalMeshComponent}
 TailMesh {type=SkeletalMeshComponent}
 UpArm_Spine_Rot {type=number}
 UpArm_Spine_Rot_R {type=number}
]
