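# Animation Blueprint for the DracoDueler face accessory, making its jaw follow the character's head.
# It copies the attached parent and pawn Head mesh poses, blending faceAttach and C_jaw from the head.
# dyn_jaw is rotation-constrained to C_jaw, offset additively, then cached.
# In the front end, the Frontend_Idle_Bypass curve weights a DefaultSlot layer over that pose from dyn_jaw_mstr down, for custom idles.
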AnimBlueprint /BRCosmetics/Characters/Player/Male/Medium/Bodies/M_MED_DracoDueler/Meshes/Parts/M_MED_DracoDueler_FaceAcc_AnimBP
target_skeleton=&/M_MED_DracoDueler_FaceAcc_Skeleton
graphs [
 AnimGraph {nodes [
   Output Pose(Result=Layered blend per bone(
     BasePose=Use cached pose 'startPose'()
     BlendPoses_0=Slot 'DefaultSlot'()
     BlendWeights_0=Get Custom Idle is Playing Float() -> (Custom Idle is Playing Float:float)
    ) {layerSetup [{branchFilters [{boneName=dyn_jaw_mstr}]}]}) {layerGroup=None}
   n1 = Blend Bone By Channel!(
    A:Pose Link Structure=Copy Pose From Mesh() {bCopyCurves=true, bCopyCustomAttributes=true, bUseAttachedParent=true, bUseMeshPose=true}
    B:Pose Link Structure=Copy Pose From Mesh(SourceMeshComponent=Get Head_Mesh() -> (Head_Mesh:Skeletal Mesh Component Object Reference)) {bCopyCurves=true, bCopyCustomAttributes=true, bUseAttachedParent=true, bUseMeshPose=true}
    Alpha:float=1
   ) -> (Pose:Pose Link Structure) {
    is=title AnimGraphNode_BlendBoneByChannel
    blendNode {boneDefinitions [
      {bBlendRotation=true, bBlendScale=true, bBlendTranslation=true, sourceBone {boneName=faceAttach}, targetBone {boneName=None}}
      {bBlendRotation=true, bBlendScale=true, bBlendTranslation=true, sourceBone {boneName=C_jaw}, targetBone {boneName=None}}
     ]}
   }
   n2 = "Constraint - dyn_jaw (R)"(ComponentPose=Local To Component(LocalPose=@n1), Alpha=1, ConstraintWeights_0=1) {
    alphaBoolBlend {blendOption=Linear}
    boneToModify {boneName=dyn_jaw}
    constraintSetup [{transformType=Rotation, targetBone {boneName=C_jaw}}]
    constraintWeights [1]
   }
   n3 = "Transform (Modify) Bone - Bone: dyn_jaw"(ComponentPose=@n2, Alpha=1, Translation="0,0,0", Rotation="12,0,0", Scale="1,1,1") {
    rotationMode=BMM_Additive
    rotationSpace=BCS_BoneSpace
    translationMode=BMM_Additive
    translationSpace=BCS_BoneSpace
    alphaBoolBlend {blendOption=Linear}
    boneToModify {boneName=dyn_jaw}
    rotation {pitch=12}
   }
   Save cached pose 'startPose'(Pose=Component To Local(ComponentPose=@n3)) {cacheName=startPose}
  ]}
 EventGraph {nodes [
   Event BlueprintUpdateAnimation()
   > Branch(Condition=Get bIsInFrontEnd())
   > Set Custom Idle is Playing Float(Custom Idle is Playing Float:float=GetCurveValue(self=GetOwnerFrontEndAnimInstance(), CurveName=Frontend_Idle_Bypass)) -> (Output_Get:float)
   Event BlueprintInitializeAnimation()
   > Set Skel_Mesh(Skel_Mesh=GetOwnerSkeletalMeshComponent())
   > n1 = Cast To PlayerPawn_Athena(Object=TryGetPawnOwner())
   > n2 = Set Pawn Owner(Pawn Owner:Fort Player Pawn Athena Object Reference=@n1) -> (Output_Get:Fort Player Pawn Athena Object Reference)
   > Set Head_Mesh(Head_Mesh:Skeletal Mesh Component Object Reference=GetSkeletalMeshForPartType(self=@n2, PartType=Head)) -> (Output_Get:Skeletal Mesh Component Object Reference)
   Get Body_Mesh()
  ]}
]
variables [
 Body_Mesh {type=SkeletalMeshComponent}
 Custom Idle is Playing Float {type=number}
 Head_Mesh {type=SkeletalMeshComponent}
 Pawn Owner {type=FortPlayerPawnAthena}
 Skel_Mesh {type=SkeletalMeshComponent}
]
